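# Niagara system for the LimeRock TV rift effect, a mostly cyan and blue energy rift with lightning bolts, corona glow, rays, flickering light and shockwave.
# User.Lifetime paces the Bolt emitter, whose spawn rate ramps up over its loop unless User.TestModel forces a constant rate.
# System.LifetimeAdd times most other layers, while User.Radius scales meshes, sprites and light, with curves driving dynamic material parameters.
# Several layers are disabled, including cracks, GPU embers, fragments and flare.
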NiagaraSystem /LimeRock/Assets/FX/TV_Rift/NS_LimeRock_TV_Rift
dependencies {
 usedDataInterfaces [
  &NiagaraDataInterfaceCurve
  &NiagaraDataInterfaceVector2DCurve
  &NiagaraDataInterfaceArrayColor
  &NiagaraDataInterfaceColorCurve
 ]
 usedDynamicInputs []
 usedModules [
  &self:SystemSpawnScript.SystemScriptSource.SystemScriptGraph.NiagaraNodeAssignment_0.SetVariables_#1
  &SystemState
 ]
}
emitters {
 Bolt {
  emitterUpdateScript [
   &EmitterState(
    Inactive Response=::NewEnumerator1
    Life Cycle Mode=::NewEnumerator1
    Loop Behavior=::NewEnumerator1
    Loop Duration=&Add_Float(A=$User.Lifetime, B=0)
    UseLoopDelay=true
   )
   &SpawnRate(
    SpawnRate=&MakeCustomFloatFromBool(
     Bool=$User.TestModel
     True Float=50
     False Float=&FloatFromCurve(
      FloatCurve=DataInterface() {propertyValues {Curve {keys [
          {arriveTangent=0.4322, interpMode=RCIM_Cubic, leaveTangent=0.4322, leaveTangentWeight=0.2, tangentMode=RCTM_User, time=0.3, value=0}
          {arriveTangent=2.4687, arriveTangentWeight=43.464, interpMode=RCIM_Cubic, leaveTangent=2.4687, tangentMode=RCTM_User, time=0.95, value=0.85}
         ]}}}
      CurveIndex=$Emitter.NormalizedLoopAge
      Scale Curve=30
     )
    )
    Use Spawn Probability=false
   )
  ]
  particleSpawnScript [
   /SetVariables(Particles.MeshIndex=&RandomRangeInt(Minimum=0, Maximum=3)) {isSetParameters=true}
   &InitializeParticle(
    Color=&SelectLinearColorFromArray(
     Array Sampling Mode=::NewEnumerator0
     Direct Array Index=?
     Color Selection Array=DataInterface() {propertyValues {ColorData [<rgba 0 6 6 1>, <rgba 0 1.3333 2 1>]}}
     Random Seed=0
    )
    Lifetime=&RandomRangeFloat(Minimum=0.23, Maximum=0.28)
    Mesh Scale=&Multiply_VectorByFloat(Vector=<0.5 0.5 0.7>, Float=0.8)
    Mesh Scale Mode=::NewEnumerator4
    Sprite Size Mode=::NewEnumerator3
    Uniform Sprite Size=800
    UsePositionOffset=false
    Write Material Random=true
   )
   &InitialMeshOrientation(Use Rotation Vector=false)
   &CameraOffset(Camera Offset Amount=50)
   &ShapeLocation(MaskSpawnGroup=false, Override Seed=false, Sphere Radius=10, Sphere Surface Distribution=1, UseOwnerScale=true)
  ]
  particleUpdateScript [
   &SolveForcesAndVelocity(Clamp Velocity=false, Limit Acceleration=false)
   &ParticleState()
   &ScaleMeshSize(Scale Factor=&Multiply_VectorByFloat(Vector=<1 1 1>, Float=&Multiply_Float(A=1, B=$User.Radius)))
   &ScaleColor(
    Scale Alpha=&FloatFromCurve(FloatCurve=DataInterface() {propertyValues {Curve {keys [
         {interpMode=RCIM_Cubic, tangentMode=RCTM_User, time=0, value=0}
         {interpMode=RCIM_Cubic, tangentMode=RCTM_User, time=0.5, value=1}
         {tangentMode=RCTM_User, time=1, value=0}
        ]}}})
    Scale RGB=&VectorFromFloat(Value=&FloatFromCurve(
      FloatCurve=DataInterface() {propertyValues {Curve {keys [
          {interpMode=RCIM_Cubic, tangentMode=RCTM_User, time=0, value=0}
          {arriveTangent=-3.0443, interpMode=RCIM_Cubic, leaveTangent=-3.0443, tangentMode=RCTM_User, time=0.1, value=0.95}
          {interpMode=RCIM_Cubic, tangentMode=RCTM_User, time=0.4, value=0.55}
          {tangentMode=RCTM_User, time=1, value=0}
         ]}}}
      Scale Curve=&Multiply_Float(A=15, B=1)
     ))
    ScaleA=true
    ScaleRGB=true
   )
   &DynamicMaterialParameters(
    Index 0 Param 1=&Multiply_Float(A=2, B=$User.Radius)
    Index 0 Param 2=&FloatFromCurve(
     FloatCurve=DataInterface() {propertyValues {Curve {keys [
         {interpMode=RCIM_Cubic, tangentMode=RCTM_User, time=0, value=0}
         {arriveTangent=-1.052, interpMode=RCIM_Cubic, leaveTangent=-1.052, tangentMode=RCTM_User, time=0.4, value=0.95}
         {arriveTangent=-2.3073, interpMode=RCIM_Cubic, leaveTangent=-2.3073, tangentMode=RCTM_User, time=0.9, value=0}
        ]}}}
     Scale Curve=&RandomRangeFloat(Minimum=3, Maximum=2)
    )
    Index 0 Param 3=&FloatFromCurve(
     FloatCurve=DataInterface() {propertyValues {Curve {keys [
         {interpMode=RCIM_Cubic, tangentMode=RCTM_User, time=0, value=0}
         {tangentMode=RCTM_User, time=1.5, value=0.85}
         {interpMode=RCIM_Cubic, tangentMode=RCTM_User, time=2.5, value=1}
        ]}}}
     Scale Curve=&RandomRangeFloat(Minimum=4, Maximum=3)
    )
    Index 0 Param 4=1
    Index 1 Param 1=&FloatFromCurve(FloatCurve=DataInterface() {propertyValues {Curve {keys [{interpMode=RCIM_Cubic, tangentMode=RCTM_User, time=0, value=1}, {interpMode=RCIM_Cubic, tangentMode=RCTM_User, time=1, value=0}]}}})
    Index 1 Param 2=50
    Index 1 Param 3=1
    Index 1 Param 4=1
    Index 2 Param 1=&FloatFromCurve(FloatCurve=DataInterface() {propertyValues {Curve {keys [{interpMode=RCIM_Cubic, tangentMode=RCTM_User, time=0, value=1}, {tangentMode=RCTM_User, time=1, value=0.5}]}}})
    Module.Param0WriteEnabled=true
    Module.Param0WriteEnabled1=true
    Module.Param0WriteEnabled2=true
    Module.Param1WriteEnabled=true
    Module.Param1WriteEnabled1=true
    Module.Param1WriteEnabled2=true
    Module.Param2WriteEnabled=true
    Module.Param2WriteEnabled1=true
    Module.Param2WriteEnabled2=true
    Module.Param3WriteEnabled=true
    Module.Param3WriteEnabled1=true
    Module.Param3WriteEnabled2=true
    Write Parameter Index 1=true
    Write Parameter Index 2=true
   )
  ]
  rendererClasses [&NiagaraMeshRendererProperties]
  renderers [{rendererClass=&NiagaraMeshRendererProperties, rendererIndex=0}]
 }
 Corona {
  bEnabled=false
  emitterUpdateScript [
   &EmitterState()
   &SpawnBurst_Instantaneous(Spawn Count=1, Use Spawn Probability=false)
  ]
  particleSpawnScript [
   &InitializeParticle(
    Color=<rgba 0 2.6667 4 0.5>
    Lifetime=$System.LifetimeAdd
    Sprite Size Mode=::NewEnumerator3
    Uniform Sprite Size=650
    UsePositionOffset=false
    Write Material Random=true
   )
   &CameraOffset(Camera Offset Amount=1)
  ]
  particleUpdateScript [
   &ParticleState()
   &SolveForcesAndVelocity(Clamp Velocity=false, Limit Acceleration=false)
   &DynamicMaterialParameters(
    Index 0 Param 1=&FloatFromCurve(FloatCurve=DataInterface() {propertyValues {Curve {keys [{arriveTangent=-2.7401, interpMode=RCIM_Cubic, leaveTangent=-2.7401, tangentMode=RCTM_User, time=0, value=2}, <at 1 0.7 RCIM_Cubic>]}}})
    Index 0 Param 2=0.4
    Index 0 Param 3=0.6
    Index 0 Param 4=1
    Module.Param0WriteEnabled=true
    Module.Param1WriteEnabled=true
    Module.Param2WriteEnabled=true
    Module.Param3WriteEnabled=true
   )
   &ScaleColor(
    Scale Alpha=&FloatFromCurve(FloatCurve=DataInterface() {propertyValues {Curve {keys [
         <at 0 0 RCIM_Cubic>
         {arriveTangent=2.8165, interpMode=RCIM_Cubic, leaveTangent=2.8165, tangentMode=RCTM_User, time=0.9, value=1}
         {arriveTangent=-4.8514, interpMode=RCIM_Cubic, leaveTangent=-4.8514, tangentMode=RCTM_User, time=1, value=0}
        ]}}})
    Scale RGB=&MakeVectorFromLinearColorRGB(Color=<rgba 1 1 1 1>)
    ScaleA=true
    ScaleRGB=true
   )
   &FN_Light_Attributes(
    Color=$Particles.Color
    Create Color Attribute=true
    Create Exponent Attribute=true
    Create Position Attribute=true
    Create Radius Attribute=true
    Exponent=4
    Position=&Add_Vector(A=Linked() {linkedVariable {name=Particles.Position, type=NiagaraPosition}}, B=<0 0 0>)
    Radius=400
    Scale Color=3
   )
  ]
  rendererClasses [&NiagaraSpriteRendererProperties]
  renderers [{rendererClass=&NiagaraSpriteRendererProperties}]
 }
 Corona001 {
  emitterUpdateScript [
   &EmitterState()
   &SpawnBurst_Instantaneous(Spawn Count=1, Use Spawn Probability=false)
  ]
  particleSpawnScript [
   &InitializeParticle(
    Color=<rgba 3 3 3 0.5>
    Lifetime=&Add_Float(A=$System.LifetimeAdd, B=0.1)
    Sprite Size Mode=::NewEnumerator3
    Uniform Sprite Size=450
    UsePositionOffset=false
    Write Material Random=true
   )
   &CameraOffset(Camera Offset Amount=1)
   &SystemLocation(MaskSpawnGroup=false, Offset=<0 0 120>) {enabled=false}
  ]
  particleUpdateScript [
   &ParticleState()
   &ScaleSpriteSize(
    Uniform Curve Scale=&Multiply_Float(A=0.7, B=$User.Radius)
    Uniform Curve Sprite Scale=DataInterface() {propertyValues {Curve {keys [{tangentMode=RCTM_User, time=0, value=1}, {tangentMode=RCTM_User, time=1, value=1}]}}}
   )
   &SolveForcesAndVelocity(Clamp Velocity=false, Limit Acceleration=false)
   &DynamicMaterialParameters(
    Index 0 Param 1=&FloatFromCurve(FloatCurve=DataInterface() {propertyValues {Curve {keys [
         {arriveTangent=-0.1181, interpMode=RCIM_Cubic, leaveTangent=-0.1181, tangentMode=RCTM_User, time=0, value=2}
         {arriveTangent=0.8207, interpMode=RCIM_Cubic, leaveTangent=0.8207, tangentMode=RCTM_User, time=0.985, value=1}
         {arriveTangent=-47.1788, interpMode=RCIM_Cubic, leaveTangent=-47.1788, tangentMode=RCTM_User, time=1, value=0.5}
        ]}}})
    Index 0 Param 2=0.4
    Index 0 Param 3=0.6
    Index 0 Param 4=1
    Module.Param0WriteEnabled=true
    Module.Param1WriteEnabled=true
    Module.Param2WriteEnabled=true
    Module.Param3WriteEnabled=true
   )
   &ScaleColor(
    Scale Alpha=&FloatFromCurve(FloatCurve=DataInterface() {propertyValues {Curve {keys [<at 0 0 RCIM_Cubic>,<at 0.25 1 RCIM_Cubic>,<at 1 0 RCIM_Cubic>]}}})
    Scale RGB=&MakeVectorFromLinearColorRGB(Color=<rgba 1 1 1 1>)
    ScaleA=true
    ScaleRGB=true
   ) {enabled=false}
   &ScaleColor(
    Scale Alpha=&FloatFromCurve(FloatCurve=DataInterface() {propertyValues {Curve {keys [
         {arriveTangent=0.7534, interpMode=RCIM_Cubic, leaveTangent=0.7534, tangentMode=RCTM_User, time=0, value=0}
         {arriveTangent=0.089, interpMode=RCIM_Cubic, leaveTangent=0.089, tangentMode=RCTM_User, time=0.9, value=0.6}
         {arriveTangent=-4.8514, interpMode=RCIM_Cubic, leaveTangent=-4.8514, tangentMode=RCTM_User, time=1, value=0}
        ]}}})
    Scale RGB=&MakeVectorFromLinearColorRGB(Color=<rgba 1 1 1 1>)
    ScaleA=true
    ScaleRGB=true
   ) {module=ScaleColor001}
   &FN_Light_Attributes(
    Color=$Particles.Color
    Create Color Attribute=true
    Create Exponent Attribute=true
    Create Position Attribute=true
    Create Radius Attribute=true
    Exponent=4
    Position=&Add_Vector(A=Linked() {linkedVariable {name=Particles.Position, type=NiagaraPosition}}, B=<0 0 0>)
    Radius=400
    Scale Color=3
   )
  ]
  rendererClasses [&NiagaraSpriteRendererProperties]
  renderers [{rendererClass=&NiagaraSpriteRendererProperties}]
 }
 Crack01 {
  bEnabled=false
  emitterUpdateScript [
   &EmitterState()
   &SpawnBurst_Instantaneous(Spawn Count=1, Use Spawn Probability=false)
  ]
  particleSpawnScript [
   /Niagara/Modules/Spawn/Initialization/InitializeParticle(
    Color=<rgba 0.75 0.8333 1 1>
    Lifetime=$System.LifetimeAdd
    Mesh Scale=&VectorFromFloat(Value=7)
    Module.Write Color=true
    Module.Write Lifetime=true
    Module.Write Mass=true
    Module.Write Material Random=true
    Module.Write Position=false
    Module.Write Scale=true
    Module.Write SpriteRotation=false
    Module.Write SpriteSize=false
   )
   /SetVariables(Particles.MeshOrientation=&MakeQuatFromAxisAngle(
     Axis=<1 1 1>
     AngleInDegrees=&MaskFloatBySpawnGroup(
      Masked Float=&RandomRangeFloat(Minimum=-80, Maximum=0)
      Passthrough Float=&RandomRangeFloat(Minimum=80, Maximum=180)
     )
    )) {isSetParameters=true}
  ]
  particleUpdateScript [
   &ParticleState()
   &DynamicMaterialParameters(
    Index 0 Param 1=&FloatFromCurve(FloatCurve=DataInterface() {propertyValues {Curve {keys [
         {interpMode=RCIM_Cubic, tangentMode=RCTM_User, time=0, value=0}
         {arriveTangent=2.5215, interpMode=RCIM_Cubic, leaveTangent=2.5215, tangentMode=RCTM_User, time=1, value=1}
        ]}}})
    Index 0 Param 2=&FloatFromCurve(FloatCurve=DataInterface() {propertyValues {Curve {keys [<at 0.4 0.2 RCIM_Cubic>,<at 0.9993 0.5 RCIM_Cubic>]}}})
    Index 0 Param 3=1
    Index 0 Param 4=&FloatFromCurve(FloatCurve=DataInterface() {propertyValues {Curve {keys [<at 0 1 RCIM_Cubic>,<at 1 1>]}}})
    Module.Param0WriteEnabled=true
    Module.Param1WriteEnabled=true
    Module.Param2WriteEnabled=true
    Module.Param3WriteEnabled=true
   )
   &CameraOffset(Camera Offset Amount=100)
  ]
  rendererClasses [&NiagaraMeshRendererProperties]
  renderers [{rendererClass=&NiagaraMeshRendererProperties, rendererIndex=0}]
 }
 Crack02 {
  bEnabled=false
  emitterUpdateScript [
   &EmitterState()
   &SpawnBurst_Instantaneous(Use Spawn Probability=false, UseLoopCountLimit=false)
  ]
  particleSpawnScript [
   /Niagara/Modules/Spawn/Initialization/V2/InitializeParticle(
    Color=<rgba 0 2 3 1>
    Lifetime=$System.LifetimeAdd
    Mesh Scale Mode=::NewEnumerator3
    Mesh Uniform Scale=4
    UsePositionOffset=false
    Write Material Random=true
   )
   &ShapeLocation(MaskSpawnGroup=false, Offset=$System.OffsetCrack, Offset Mode=::NewEnumerator0, Override Seed=false, Sphere Radius=0, UseOwnerScale=true)
   &InitialMeshOrientation(
    Mesh Orientation Mode=::NewEnumerator2
    Orientation Vector=<0 0 1>
    Rotation=&RandomRangeVector(Minimum=<0 0 0>, Maximum=<0 0 1>)
    Use Rotation Vector=true
   )
  ]
  particleUpdateScript [
   &ParticleState()
   &ScaleMeshSize(Scale Factor=&Multiply_VectorByFloat(
     Vector=<2 2 2>
     Float=&FloatFromCurve(FloatCurve=DataInterface() {propertyValues {Curve {keys [
          {interpMode=RCIM_Cubic, tangentMode=RCTM_User, time=0, value=0.8}
          {arriveTangent=0.9153, interpMode=RCIM_Cubic, tangentMode=RCTM_User, time=0.98, value=1}
          {interpMode=RCIM_Cubic, tangentMode=RCTM_User, time=0.9925, value=1}
         ]}}})
    ))
   &DynamicMaterialParameters(
    Index 0 Param 1=&FloatFromCurve(
     FloatCurve=DataInterface() {propertyValues {Curve {keys [
         {interpMode=RCIM_Cubic, tangentMode=RCTM_User, time=0, value=0}
         {arriveTangent=2.5215, interpMode=RCIM_Cubic, leaveTangent=2.5215, tangentMode=RCTM_User, time=0.95, value=1}
         {arriveTangent=-36.0646, interpMode=RCIM_Cubic, leaveTangent=-36.0646, tangentMode=RCTM_User, time=1, value=0}
        ]}}}
     Scale Curve=0.5
    )
    Index 0 Param 2=0.3
    Module.Param0WriteEnabled=true
    Module.Param1WriteEnabled=true
    Module.Param2WriteEnabled=true
    Module.Param3WriteEnabled=true
   )
  ]
  rendererClasses [&NiagaraMeshRendererProperties]
  renderers [{rendererClass=&NiagaraMeshRendererProperties, rendererIndex=0}]
 }
 Embers {
  bEnabled=false
  simTarget=GPUComputeSim
  emitterUpdateScript [
   &EmitterState()
   &SpawnRate(
    SpawnRate=&Multiply_Float(
     A=&FloatFromCurve(
      FloatCurve=DataInterface() {propertyValues {Curve {keys [
          <at 0 0 RCIM_Cubic>
          {arriveTangent=0.4926, interpMode=RCIM_Cubic, leaveTangent=0.4926, tangentMode=RCTM_User, time=0.75, value=0.15}
          {arriveTangentWeight=43.464, interpMode=RCIM_Cubic, tangentMode=RCTM_User, time=0.9, value=1}
         ]}}}
      CurveIndex=$Emitter.NormalizedLoopAge
      Scale Curve=400
     )
     B=&FloatFromCurve(
      FloatCurve=DataInterface() {propertyValues {Curve {keys [{interpMode=RCIM_Cubic, tangentMode=RCTM_Break, time=0, value=1}, <at 15000 0.005>]}}}
      CurveIndex=$Engine.Owner.LODDistance
     )
    )
    Use Spawn Probability=false
   )
  ]
  particleSpawnScript [
   /Niagara/Modules/Spawn/Initialization/InitializeParticle(
    Color=<rgba 0 24.7 25 1>
    Lifetime=&RandomRangeFloat(Minimum=0.5, Maximum=1.2, Evaluation Type=::NewEnumerator1)
    Module.Write Color=true
    Module.Write Lifetime=true
    Module.Write Mass=true
    Module.Write Material Random=true
    Module.Write Position=false
    Module.Write Scale=false
    Module.Write SpriteRotation=false
    Module.Write SpriteSize=true
    Sprite Size=&RandomRangeVector2D(Minimum=<4 4>, Maximum=<8 8>, Evaluation Type=::NewEnumerator1)
   )
   &ShapeLocation(
    MaskSpawnGroup=false
    Override Seed=false
    Sphere Radius=&Lerp_Float(
     A=100
     B=500
     Alpha=&FloatFromCurve(
      FloatCurve=DataInterface() {propertyValues {Curve {keys [
          <at 0 0 RCIM_Cubic>
          {arriveTangent=0.4926, interpMode=RCIM_Cubic, leaveTangent=0.4926, tangentMode=RCTM_User, time=0.8, value=0.15}
          {arriveTangentWeight=43.464, interpMode=RCIM_Cubic, tangentMode=RCTM_User, time=0.95, value=1}
         ]}}}
      CurveIndex=$Emitter.NormalizedLoopAge
     )
    )
    UseOwnerScale=true
   )
   &AddVelocity(
    Module.Use Falloff Distance=false
    Override Seed=false
    Velocity Mode=::NewEnumerator1
    Velocity Speed=&Lerp_Float(
     A=&RandomRangeFloat(Minimum=75, Maximum=500)
     B=&RandomRangeFloat(Minimum=800, Maximum=3000)
     Alpha=&FloatFromCurve(
      FloatCurve=DataInterface() {propertyValues {Curve {keys [
          <at 0 0 RCIM_Cubic>
          {arriveTangent=0.9042, interpMode=RCIM_Cubic, leaveTangent=0.9042, tangentMode=RCTM_User, time=0.9, value=0.25}
          {arriveTangentWeight=43.464, interpMode=RCIM_Cubic, tangentMode=RCTM_User, time=1, value=1}
         ]}}}
      CurveIndex=$Emitter.NormalizedLoopAge
     )
    )
   )
  ]
  particleUpdateScript [
   &ParticleState()
   &Drag(Drag=6, Use Linear Drag=true, Use Rotational Drag=false)
   &CurlNoiseForce(Noise Frequency=100, Noise Strength=1000, Pan Noise=false)
   &SolveForcesAndVelocity(Clamp Velocity=false, Limit Acceleration=false)
   &ScaleSpriteSize(Scale Factor=&Vector2DFromFloat(Value=&FloatFromCurve(FloatCurve=DataInterface() {propertyValues {Curve {keys [
          <at -0.0044 0.2654>
          {arriveTangent=0.7558, interpMode=RCIM_Cubic, leaveTangent=0.7558, time=0.157, value=0.5261}
          {arriveTangent=-0.6241, interpMode=RCIM_Cubic, leaveTangent=-0.6241, time=0.7034, value=0.8003}
          <at 1 0>
         ]}}})))
   &ScaleSpriteSizeBySpeed(Max Scale Factor=<1 12>, Source Velocity:Vector3f=&Multiply_Vector(A=<0 0.01 1>, B=$Particles.Velocity))
   &ScaleSpriteSize(Scale Factor=&MakeVector2D(
     X=&FloatFromCurve(FloatCurve=DataInterface() {propertyValues {Curve {keys [<at 0 1>,<at 15000 12>]}}}, CurveIndex=$Engine.Owner.LODDistance)
     Y=1
    )) {module=ScaleSpriteSize001}
   /SetVariables(
    Particles.Color2=&MultiplyLinearColor(Linear Color=$Particles.Color, Scale Factor=<0 0 0 10>)
    Particles.Size2:Vector2f=&Multiply_Vector2D(A=$Particles.SpriteSize, B=<1.7 1.2>)
   ) {isSetParameters=true}
  ]
  rendererClasses [&NiagaraSpriteRendererProperties]
  renderers [
   {rendererClass=&NiagaraSpriteRendererProperties}
   {rendererClass=&NiagaraSpriteRendererProperties, rendererIndex=1}
  ]
 }
 Flare {
  bEnabled=false
  emitterUpdateScript [
   &EmitterState()
   &SpawnBurst_Instantaneous(Spawn Count=1, Use Spawn Probability=false)
  ]
  particleSpawnScript [&InitializeParticle(
    Color=<rgba 0 1 2 1>
    Lifetime=$System.LifetimeAdd
    Module.Write Color=true
    Module.Write Lifetime=true
    Module.Write Mass=true
    Module.Write Material Random=true
    Module.Write Position=false
    Module.Write Scale=false
    Module.Write SpriteRotation=false
    Module.Write SpriteSize=true
    Sprite Size=&Vector2DFromFloat(Value=3000)
   )]
  particleUpdateScript [
   &ParticleState()
   &ScaleSpriteSize(Uniform Curve Sprite Scale=DataInterface() {propertyValues {Curve {keys [
        {interpMode=RCIM_Cubic, tangentMode=RCTM_User, time=0, value=0.3}
        {arriveTangent=0.7525, interpMode=RCIM_Cubic, leaveTangent=0.7525, tangentMode=RCTM_User, time=0.8, value=0.45}
        {interpMode=RCIM_Cubic, tangentMode=RCTM_User, time=0.9, value=0.7}
       ]}}})
   &ScaleColor(
    Scale Alpha=&FloatFromCurve(FloatCurve=DataInterface() {propertyValues {Curve {keys [
         <at 0 0 RCIM_Cubic>
         {arriveTangent=0.289, interpMode=RCIM_Cubic, leaveTangent=0.289, time=0.4918, value=0.248}
         {arriveTangent=3.4072, interpMode=RCIM_Cubic, leaveTangent=3.4072, time=0.5041, value=0.1457}
         {arriveTangent=-2.7103, interpMode=RCIM_Cubic, leaveTangent=-2.7103, time=0.5232, value=0.3549}
         {arriveTangent=2.94, interpMode=RCIM_Cubic, leaveTangent=2.94, time=0.5374, value=0.0554}
         {arriveTangent=0.8123, interpMode=RCIM_Cubic, leaveTangent=0.8123, time=0.5612, value=0.4666}
         {arriveTangent=3.9277, interpMode=RCIM_Cubic, leaveTangent=3.9277, time=0.5922, value=0.0999}
         {arriveTangent=3.855, interpMode=RCIM_Cubic, leaveTangent=3.855, time=0.6114, value=0.6636}
         {arriveTangent=-6.193, interpMode=RCIM_Cubic, leaveTangent=-6.193, time=0.6303, value=0.2469}
         {arriveTangent=1.1094, interpMode=RCIM_Cubic, leaveTangent=1.1094, time=0.6459, value=0.4498}
         {arriveTangent=3.878, interpMode=RCIM_Cubic, leaveTangent=3.878, time=0.6893, value=0.3124}
         {arriveTangent=0.3841, interpMode=RCIM_Cubic, leaveTangent=0.3841, time=0.6979, value=0.6518}
         {arriveTangent=1.6385, interpMode=RCIM_Cubic, leaveTangent=1.6385, time=0.7238, value=0.3256}
         {arriveTangent=1.2094, interpMode=RCIM_Cubic, leaveTangent=1.2094, time=0.7388, value=0.7187}
         {arriveTangent=5.1593, interpMode=RCIM_Cubic, leaveTangent=5.1593, time=0.769, value=0.3803}
         {arriveTangent=2.2949, interpMode=RCIM_Cubic, leaveTangent=2.2949, time=0.7839, value=0.9515}
         {arriveTangent=-4.6065, interpMode=RCIM_Cubic, leaveTangent=-4.6065, time=0.815, value=0.4859}
         {arriveTangent=9.5468, interpMode=RCIM_Cubic, leaveTangent=9.5468, time=0.8665, value=0.5711}
         {arriveTangent=0.1948, interpMode=RCIM_Cubic, leaveTangent=0.1948, time=0.8707, value=1.017}
         {arriveTangent=-7.1532, interpMode=RCIM_Cubic, leaveTangent=-7.1532, time=0.8934, value=0.5764}
         {arriveTangent=15.7037, interpMode=RCIM_Cubic, leaveTangent=15.7037, time=0.925, value=0.6286}
         {arriveTangent=-2.0814, interpMode=RCIM_Cubic, leaveTangent=-2.0814, time=0.9294, value=1.1414}
         {arriveTangent=-16.7837, interpMode=RCIM_Cubic, leaveTangent=-16.7837, time=0.9479, value=0.5809}
         {arriveTangent=31.4448, interpMode=RCIM_Cubic, leaveTangent=31.4448, time=0.9612, value=0.6074}
         {arriveTangent=2.7173, interpMode=RCIM_Cubic, leaveTangent=2.7173, time=0.9678, value=1.2074}
         {arriveTangent=-6.4425, interpMode=RCIM_Cubic, leaveTangent=-6.4425, time=0.9866, value=0.6763}
         <at 1 1>
        ]}}})
    Scale RGB=<0.5 0.5 0.5>
    ScaleA=true
    ScaleRGB=true
   )
   &FN_Light_Attributes(
    Color=<rgba 0.05 0.3667 1 1>
    Create Color Attribute=true
    Create Exponent Attribute=true
    Create Radius Attribute=true
    Radius=&Lerp_Float(
     A=800
     B=3000
     Alpha=&FloatFromCurve(FloatCurve=DataInterface() {propertyValues {Curve {keys [<at 0 0>, {interpMode=RCIM_Cubic, tangentMode=RCTM_User, time=0.5, value=0.2}, <at 0.9 1>]}}})
    )
    Scale Color=8
   )
  ]
  rendererClasses [&NiagaraSpriteRendererProperties]
  renderers [{rendererClass=&NiagaraSpriteRendererProperties}]
 }
 FlareLightShockwave {
  emitterSpawnScript [/SetVariables(Emitter.DepthFade=100) {isSetParameters=true}]
  emitterUpdateScript [
   &EmitterState(Life Cycle Mode=::NewEnumerator1, Loop Behavior=::NewEnumerator1, Loop Delay=$System.LifetimeAdd, Loop Duration=1, UseLoopDelay=false)
   &SpawnBurst_Instantaneous(Use Spawn Probability=false, UseLoopCountLimit=false)
  ]
  particleSpawnScript [
   /Niagara/Modules/Spawn/Initialization/V2/InitializeParticle(
    Color=<rgba 0.1598 0.398 1 1>
    Lifetime=&Add_Float(A=$System.LifetimeAdd, B=0.5)
    Mesh Scale Mode=::NewEnumerator3
    Mesh Uniform Scale=&Divide_Float(A=$User.Radius, B=150)
    Position Offset=<15 0 0>
    Sprite Size Mode=::NewEnumerator3
    Uniform Sprite Size=2500
    UsePositionOffset=false
    Write Material Random=true
   )
   &ShapeLocation(MaskSpawnGroup=false, Offset=$System.Offset, Offset Mode=::NewEnumerator0, Override Seed=false, Sphere Radius=0, UseOwnerScale=true)
  ]
  particleUpdateScript [
   &ParticleState()
   &CameraOffset(Camera Offset Amount=&FloatFromCurve(
     FloatCurve=DataInterface() {propertyValues {Curve {keys [
         {arriveTangent=4.0899, interpMode=RCIM_Cubic, leaveTangent=4.0899, tangentMode=RCTM_User, time=0, value=0.5}
         {arriveTangent=-3.6426, interpMode=RCIM_Cubic, leaveTangent=-3.6426, tangentMode=RCTM_User, time=0.2, value=1}
         {arriveTangent=5.165, interpMode=RCIM_Cubic, leaveTangent=5.165, tangentMode=RCTM_User, time=0.4, value=0.3}
         <at 0.75 0>
        ]}}}
     Scale Curve=200
    ))
   &ScaleMeshSize(Scale Factor=&VectorFromFloat(Value=&FloatFromCurve(
      FloatCurve=DataInterface() {propertyValues {Curve {keys [
          {arriveTangent=0.0946, interpMode=RCIM_Cubic, leaveTangent=0.0946, tangentMode=RCTM_User, time=0, value=0.1}
          {
           arriveTangent=1.5424
           arriveTangentWeight=0.1302
           interpMode=RCIM_Cubic
           leaveTangent=1.5424
           leaveTangentWeight=0.2603
           tangentMode=RCTM_User
           time=0.8
           value=0.4145
          }
          {arriveTangent=-0.3632, interpMode=RCIM_Cubic, leaveTangent=-0.3632, tangentMode=RCTM_User, time=1, value=1}
         ]}}}
      CurveIndex=$Particles.Age
      Scale Curve=&Multiply_Float(A=1, B=$User.Radius)
     )))
   &DynamicMaterialParameters(
    Index 0 Param 1=1
    Index 0 Param 2=&Lerp_Float(
     A=1
     B=0.3
     Alpha=&FloatFromCurve(FloatCurve=DataInterface() {propertyValues {Curve {keys [
          {arriveTangent=5.1433, interpMode=RCIM_Cubic, time=0, value=0}
          {arriveTangent=4.5726, interpMode=RCIM_Cubic, leaveTangent=4.5726, tangentMode=RCTM_User, time=0.25, value=0.7}
          {arriveTangent=-6.4921, interpMode=RCIM_Cubic, leaveTangent=2.9849, tangentMode=RCTM_Break, time=0.4, value=0.35}
          {arriveTangent=0.6574, interpMode=RCIM_Cubic, leaveTangent=0.6574, tangentMode=RCTM_User, time=1, value=1}
         ]}}})
    )
    Index 3 Param 1=&Multiply_Float(A=&MakeFloatFromVector(Vector=$Particles.Scale, Channel=::NewEnumerator0), B=90)
    Index 3 Param 2=&FloatFromCurve(FloatCurve=DataInterface() {propertyValues {Curve {keys [
         {arriveTangent=0.6249, interpMode=RCIM_Cubic, leaveTangent=0.6249, tangentMode=RCTM_User, time=0, value=0}
         {arriveTangent=1.3259, interpMode=RCIM_Cubic, leaveTangent=1.3259, tangentMode=RCTM_User, time=0.95, value=0.2765}
         {interpMode=RCIM_Cubic, leaveTangent=0.2859, time=1, value=1}
        ]}}})
    Module.Param0WriteEnabled=true
    Module.Param0WriteEnabled3=true
    Module.Param1WriteEnabled=true
    Module.Param1WriteEnabled3=true
    Module.Param2WriteEnabled=false
    Module.Param2WriteEnabled3=false
    Module.Param3WriteEnabled=false
    Module.Param3WriteEnabled3=false
    Write Parameter Index 3=true
   )
   &ScaleSpriteSize(Uniform Curve Sprite Scale=DataInterface() {propertyValues {Curve {keys [
        {arriveTangent=-6.1148, interpMode=RCIM_Cubic, leaveTangent=-6.1148, tangentMode=RCTM_User, time=0, value=1}
        {arriveTangent=-0.4947, interpMode=RCIM_Cubic, leaveTangent=8.7488, tangentMode=RCTM_Break, time=0.4, value=0.3}
        {
         arriveTangent=-0.2276
         arriveTangentWeight=0.0565
         interpMode=RCIM_Cubic
         leaveTangent=-0.2276
         leaveTangentWeight=0.1797
         tangentMode=RCTM_User
         time=0.5
         value=0.9
        }
        {interpMode=RCIM_Cubic, leaveTangent=-0.9894, time=0.7, value=0.5}
       ]}}})
   &ScaleColor(
    Scale Alpha=&FloatFromCurve(FloatCurve=DataInterface() {propertyValues {Curve {keys [
         {arriveTangent=18.2404, interpMode=RCIM_Cubic, leaveTangent=18.2404, tangentMode=RCTM_User, time=0, value=1}
         {arriveTangent=-1.4815, interpMode=RCIM_Cubic, leaveTangent=-1.4815, time=0.12, value=2}
         {
          arriveTangent=-4.2653
          arriveTangentWeight=0.0565
          interpMode=RCIM_Cubic
          leaveTangent=-4.2653
          leaveTangentWeight=0.1797
          tangentMode=RCTM_User
          time=0.27
          value=0.6
         }
         {arriveTangent=-11.3751, interpMode=RCIM_Cubic, leaveTangent=-11.3751, tangentMode=RCTM_User, time=0.8, value=2}
         {interpMode=RCIM_Cubic, tangentMode=RCTM_User, time=1, value=0}
        ]}}})
    ScaleA=true
    ScaleRGB=true
   )
   /SetVariables(Particles.WaveColor:LinearColor=&ColorFromCurve(ColorCurve=DataInterface() {propertyValues {
       AlphaCurve {keys [<at 0 0>,<at 0.2053 0.01 RCIM_Cubic>,<at 0.9224 0.1>,<at 1 0 RCIM_Cubic>]}
       BlueCurve {keys [{arriveTangent=10.1654, interpMode=RCIM_Cubic, time=0.331, value=10}, <at 0.9845 10>]}
       GreenCurve {keys [{arriveTangent=0.9403, interpMode=RCIM_Cubic, time=0.331, value=2.0833}, <at 0.9845 0.925>]}
       RedCurve {keys [{arriveTangent=0.1017, interpMode=RCIM_Cubic, time=0.331, value=0.5}, <at 0.9845 0.1>]}
      }})) {isSetParameters=true}
   /SetVariables(Particles.LightColor=&MultiplyLinearColor(Linear Color=$Particles.Color, Scale Factor=&MakeVector4(X=10, Y=10, Z=10, W=15))) {isSetParameters=true}
   /SetVariables(Particles.LightRadius=&FloatFromCurve(FloatCurve=DataInterface() {propertyValues {Curve {keys [<at 0.2 0.35 RCIM_Cubic>,<at 0.5 1>]}}}, Scale Curve=1000)) {isSetParameters=true}
  ]
  rendererClasses [
   &NiagaraSpriteRendererProperties
   &NiagaraLightRendererProperties
   &NiagaraMeshRendererProperties
  ]
  renderers [
   {rendererClass=&NiagaraSpriteRendererProperties}
   {rendererClass=&NiagaraLightRendererProperties, rendererIndex=1}
   {rendererClass=&NiagaraMeshRendererProperties, rendererIndex=2}
  ]
 }
 Fragments {
  bEnabled=false
  emitterUpdateScript [
   &EmitterState()
   &SpawnBurst_Instantaneous(Spawn Count=12, Use Spawn Probability=false, UseLoopCountLimit=false)
  ]
  particleSpawnScript [
   /SetVariables(Particles.MeshRand:NiagaraInt32=&RandomRangeInt(Minimum=0, Maximum=2)) {isSetParameters=true}
   &InitializeParticle(
    Color=<rgba 0 10 10 1>
    Lifetime=$System.LifetimeAdd
    Mesh Scale Mode=::NewEnumerator1
    Mesh Uniform Scale Max=8
    Mesh Uniform Scale Min=5
    UsePositionOffset=false
    Write Material Random=true
   )
   &ShapeLocation(
    MaskSpawnGroup=false
    Override Seed=false
    Sphere Radius=&Lerp_Float(
     A=40
     B=200
     Alpha=&FloatFromCurve(
      FloatCurve=DataInterface() {propertyValues {Curve {keys [
          {interpMode=RCIM_Cubic, tangentMode=RCTM_User, time=0, value=0}
          {arriveTangent=0.6613, interpMode=RCIM_Cubic, leaveTangent=0.6613, tangentMode=RCTM_User, time=0.8, value=0.175}
          {arriveTangentWeight=0.4215, interpMode=RCIM_Cubic, tangentMode=RCTM_User, time=1, value=0.5}
         ]}}}
      CurveIndex=&RandomRangeFloat(Minimum=0, Maximum=1)
     )
    )
    Sphere Surface Distribution=1
    UseOwnerScale=true
   )
   &InitialMeshOrientation(
    Mesh Orientation Mode=::NewEnumerator2
    Orientation Vector=<0 0 1>
    Rotation=&RandomRangeVector(Minimum=<0 0 0>, Maximum=<0 0.25 1>)
    Use Rotation Vector=true
   )
  ]
  particleUpdateScript [
   &ParticleState()
   &ScaleMeshSize(Scale Factor=&VectorFromFloat(Value=&FloatFromCurve(FloatCurve=DataInterface() {propertyValues {Curve {keys [<at 0.05 0>,<at 0.5 1>]}}})))
   &UpdateMeshOrientation(
    Rotation Rate=&FloatFromCurve(
     FloatCurve=DataInterface() {propertyValues {Curve {keys [
         {interpMode=RCIM_Cubic, tangentMode=RCTM_User, time=0.2, value=0.1}
         {arriveTangent=3.6747, interpMode=RCIM_Cubic, leaveTangent=3.6747, tangentMode=RCTM_User, time=1, value=1}
        ]}}}
     Scale Curve=&Multiply_Float(
      A=&RandomRangeFloat(Minimum=0.1, Maximum=0.4)
      B=&MakeCustomFloatFromBool(Bool=true, True Float=1, False Float=-1)
     )
    )
    Rotation Vector=&RandomRangeVector(Minimum=<0 0 0>, Maximum=<2 2 2>)
   )
   &PointAttractionForce(
    Attraction Radius=1000
    Attraction Strength=&FloatFromCurve(
     FloatCurve=DataInterface() {propertyValues {Curve {keys [
         {interpMode=RCIM_Cubic, tangentMode=RCTM_User, time=0, value=0.1}
         {interpMode=RCIM_Cubic, tangentMode=RCTM_User, time=0.9, value=0.3}
         {arriveTangentWeight=0.0756, interpMode=RCIM_Cubic, tangentMode=RCTM_User, time=1, value=1}
        ]}}}
     Scale Curve=&Lerp_Float(
      A=-2
      B=50
      Alpha=&FloatFromCurve(FloatCurve=DataInterface() {propertyValues {Curve {keys [<at 0 0>,<at 0.95 0>,<at 0.99 1>]}}})
     )
    )
    Kill Within Radius=false
    Module.Use Falloff=true
    UseAttractorOffset=false
   )
   &SolveForcesAndVelocity(Clamp Velocity=false, Limit Acceleration=false)
   /SetVariables(
    Particles.Color2=&MultiplyLinearColor(Linear Color=$Particles.Color, Scale Factor=<0 0 0 5>)
    Particles.MeshSize2:Vector3f=&Multiply_VectorByFloat(Vector=$Particles.Scale, Float=1.1)
    Particles.Size2:Vector2f=&Multiply_Vector2D(A=$Particles.SpriteSize, B=<2 2>)
   ) {isSetParameters=true}
  ]
  rendererClasses [&NiagaraMeshRendererProperties]
  renderers [
   {rendererClass=&NiagaraMeshRendererProperties, rendererIndex=0}
   {rendererClass=&NiagaraMeshRendererProperties, rendererIndex=1}
  ]
 }
 Light001 {
  emitterUpdateScript [
   &EmitterState()
   &SpawnBurst_Instantaneous(Spawn Count=1, Use Spawn Probability=false)
  ]
  particleSpawnScript [/Niagara/Modules/Spawn/Initialization/InitializeParticle(
    Color=<rgba 0 1 2 1>
    Lifetime=$System.LifetimeAdd
    Module.Write Color=true
    Module.Write Lifetime=true
    Module.Write Mass=true
    Module.Write Material Random=true
    Module.Write Position=false
    Module.Write Scale=false
    Module.Write SpriteRotation=false
    Module.Write SpriteSize=true
    Sprite Size=&Vector2DFromFloat(Value=11)
   )]
  particleUpdateScript [
   &ParticleState()
   &ScaleSpriteSize(Uniform Curve Sprite Scale=DataInterface() {propertyValues {Curve {keys [
        {interpMode=RCIM_Cubic, tangentMode=RCTM_User, time=0, value=0.3}
        {arriveTangent=0.7525, interpMode=RCIM_Cubic, leaveTangent=0.7525, tangentMode=RCTM_User, time=0.8, value=0.45}
        {interpMode=RCIM_Cubic, tangentMode=RCTM_User, time=0.9, value=1}
       ]}}})
   &ScaleColor(
    Scale Alpha=&FloatFromCurve(FloatCurve=DataInterface() {propertyValues {Curve {keys [
         <at 0 0 RCIM_Cubic>
         {arriveTangent=0.289, interpMode=RCIM_Cubic, leaveTangent=0.289, time=0.4918, value=0.248}
         {arriveTangent=3.4072, interpMode=RCIM_Cubic, leaveTangent=3.4072, time=0.5041, value=0.1457}
         {arriveTangent=-2.7103, interpMode=RCIM_Cubic, leaveTangent=-2.7103, time=0.5232, value=0.3549}
         {arriveTangent=2.94, interpMode=RCIM_Cubic, leaveTangent=2.94, time=0.5374, value=0.0554}
         {arriveTangent=0.8123, interpMode=RCIM_Cubic, leaveTangent=0.8123, time=0.5612, value=0.4666}
         {arriveTangent=3.9277, interpMode=RCIM_Cubic, leaveTangent=3.9277, time=0.5922, value=0.0999}
         {arriveTangent=3.855, interpMode=RCIM_Cubic, leaveTangent=3.855, time=0.6114, value=0.6636}
         {arriveTangent=-6.193, interpMode=RCIM_Cubic, leaveTangent=-6.193, time=0.6303, value=0.2469}
         {arriveTangent=1.1094, interpMode=RCIM_Cubic, leaveTangent=1.1094, time=0.6459, value=0.4498}
         {arriveTangent=3.878, interpMode=RCIM_Cubic, leaveTangent=3.878, time=0.6893, value=0.3124}
         {arriveTangent=0.3841, interpMode=RCIM_Cubic, leaveTangent=0.3841, time=0.6979, value=0.6518}
         {arriveTangent=1.6385, interpMode=RCIM_Cubic, leaveTangent=1.6385, time=0.7238, value=0.3256}
         {arriveTangent=1.2094, interpMode=RCIM_Cubic, leaveTangent=1.2094, time=0.7388, value=0.7187}
         {arriveTangent=5.1593, interpMode=RCIM_Cubic, leaveTangent=5.1593, time=0.769, value=0.3803}
         {arriveTangent=2.2949, interpMode=RCIM_Cubic, leaveTangent=2.2949, time=0.7839, value=0.9515}
         {arriveTangent=-4.6065, interpMode=RCIM_Cubic, leaveTangent=-4.6065, time=0.815, value=0.4859}
         {arriveTangent=9.5468, interpMode=RCIM_Cubic, leaveTangent=9.5468, time=0.8665, value=0.5711}
         {arriveTangent=0.1948, interpMode=RCIM_Cubic, leaveTangent=0.1948, time=0.8707, value=1.017}
         {arriveTangent=-7.1532, interpMode=RCIM_Cubic, leaveTangent=-7.1532, time=0.8934, value=0.5764}
         {arriveTangent=15.7037, interpMode=RCIM_Cubic, leaveTangent=15.7037, time=0.925, value=0.6286}
         {arriveTangent=-2.0814, interpMode=RCIM_Cubic, leaveTangent=-2.0814, time=0.9294, value=1.1414}
         {arriveTangent=-16.7837, interpMode=RCIM_Cubic, leaveTangent=-16.7837, time=0.9479, value=0.5809}
         {arriveTangent=31.4448, interpMode=RCIM_Cubic, leaveTangent=31.4448, time=0.9612, value=0.6074}
         {arriveTangent=-15.6706, interpMode=RCIM_Cubic, leaveTangent=-15.6706, time=0.9678, value=1.2074}
         <at 1 0>
        ]}}})
    Scale RGB=<0.5 0.5 0.5>
    ScaleA=true
    ScaleRGB=true
   )
   &FN_Light_Attributes(
    Color=<rgba 0.0676 0.1992 1 1>
    Create Color Attribute=true
    Create Exponent Attribute=true
    Create Radius Attribute=true
    Exponent=30
    Radius=&Lerp_Float(
     A=200
     B=5000
     Alpha=&FloatFromCurve(
      FloatCurve=DataInterface() {propertyValues {Curve {keys [<at 0 0>, {interpMode=RCIM_Cubic, tangentMode=RCTM_User, time=0.5, value=0.2}, <at 0.9 1>]}}}
      Scale Curve=&Multiply_Float(A=1, B=$User.Radius)
     )
    )
    Scale Color=8
   )
  ]
  rendererClasses [&NiagaraLightRendererProperties]
  renderers [{rendererClass=&NiagaraLightRendererProperties, rendererIndex=0}]
 }
 Rays {
  emitterUpdateScript [
   &EmitterState(Life Cycle Mode=::NewEnumerator1, Loop Behavior=::NewEnumerator1, Loop Duration=$System.LifetimeAdd, UseLoopDelay=false)
   &SpawnBurst_Instantaneous(Use Spawn Probability=false, UseLoopCountLimit=false)
   /SetVariables(Emitter.PanSpeed=0.5) {isSetParameters=true}
   /SetVariables(
    Emitter.EndColor:LinearColor=&MultiplyLinearColor(
     Linear Color=<rgba 0.2036 0.034 0.974 1>
     Scale Factor=&MakeFloat4FromVectorAndFloat("Vector (RGB)"=&VectorFromFloat(Value=40), "Float (Alpha)"=1)
    )
    Emitter.StartColor:LinearColor=&MultiplyLinearColor(
     Linear Color=<rgba 0 0.0819 0.974 1>
     Scale Factor=&MakeFloat4FromVectorAndFloat("Vector (RGB)"=&VectorFromFloat(Value=30), "Float (Alpha)"=1)
    )
   ) {isSetParameters=true}
  ]
  particleSpawnScript [
   /Niagara/Modules/Spawn/Initialization/V2/InitializeParticle(
    Color=<rgba 2 2 2 1>
    Lifetime=&Add_Float(A=$System.LifetimeAdd, B=0.5)
    Mesh Scale Mode=::NewEnumerator4
    UsePositionOffset=false
    Write Material Random=true
   )
   &ShapeLocation(MaskSpawnGroup=false, Offset=$System.Offset, Offset Mode=::NewEnumerator0, Override Seed=false, Sphere Radius=0, UseOwnerScale=true)
   &InitialMeshOrientation(Mesh Orientation Mode=::NewEnumerator2, Orientation Vector=<0 0 1>, Use Rotation Vector=true)
  ]
  particleUpdateScript [
   &ParticleState()
   &ScaleMeshSize(Scale Factor=&Multiply_VectorByFloat(Vector=<1 1 1>, Float=$User.Radius)) {enabled=false}
   &DynamicMaterialParameters(
    Index 0 Param 1=&FloatFromCurve(
     FloatCurve=DataInterface() {propertyValues {Curve {keys [<at 0 0.3 RCIM_Cubic>, {interpMode=RCIM_Cubic, tangentMode=RCTM_Break, time=0.9, value=1}]}}}
     Scale Curve=&Multiply_Float(A=0.85, B=$User.Radius)
    )
    Index 0 Param 2=&FloatFromCurve(FloatCurve=DataInterface() {propertyValues {Curve {keys [
         {arriveTangent=-0.5351, interpMode=RCIM_Cubic, leaveTangent=-0.5351, tangentMode=RCTM_User, time=0, value=1.5}
         {interpMode=RCIM_Cubic, tangentMode=RCTM_Break, time=0.9, value=3}
        ]}}})
    Index 0 Param 3=&FloatFromCurve(
     FloatCurve=DataInterface() {propertyValues {Curve {keys [
         {arriveTangent=0.0495, interpMode=RCIM_Cubic, leaveTangent=0.0495, leaveTangentWeight=0.9958, tangentMode=RCTM_User, time=0.007, value=0.2094}
         {arriveTangent=0.8804, interpMode=RCIM_Cubic, leaveTangent=0.8804, time=0.3041, value=0.9981}
         {arriveTangent=-1.1467, interpMode=RCIM_Cubic, leaveTangent=-1.1467, time=0.9, value=0.9956}
         {arriveTangent=-1.7311, interpMode=RCIM_Cubic, tangentMode=RCTM_Break, time=1, value=0.2}
        ]}}}
     Scale Curve=&Multiply_Float(A=0.85, B=$User.Radius)
    )
    Index 0 Param 4=100
    Module.Param0WriteEnabled=true
    Module.Param1WriteEnabled=true
    Module.Param2WriteEnabled=true
    Module.Param3WriteEnabled=true
   )
   &ScaleColor(
    Scale Alpha=&FloatFromCurve(
     FloatCurve=DataInterface() {propertyValues {Curve {keys [
         {interpMode=RCIM_Cubic, leaveTangentWeight=1.182, tangentMode=RCTM_User, time=0.0034, value=-0.0118}
         {
          arriveTangent=-0.9576
          arriveTangentWeight=0.3611
          interpMode=RCIM_Cubic
          leaveTangent=-0.9576
          leaveTangentWeight=0.0667
          tangentMode=RCTM_User
          time=0.9
          value=0.75
         }
         {arriveTangent=-4.8426, interpMode=RCIM_Cubic, tangentMode=RCTM_Break, time=1, value=0}
        ]}}}
     Scale Curve=0.05
    )
    Scale RGB=&MakeVectorFromLinearColorRGB(Color=&LerpLinearColors(
      StartColor=$Emitter.StartColor
      EndColor=$Emitter.EndColor
      LerpFactor=&FloatFromCurve(FloatCurve=DataInterface() {propertyValues {Curve {keys [
           {interpMode=RCIM_Cubic, leaveTangentWeight=0.771, tangentMode=RCTM_User, time=0, value=1}
           {arriveTangent=-5.6796, interpMode=RCIM_Cubic, tangentMode=RCTM_Break, time=0.5, value=0}
          ]}}})
     ))
    ScaleA=true
    ScaleRGB=true
   )
  ]
  rendererClasses [&NiagaraMeshRendererProperties]
  renderers [{rendererClass=&NiagaraMeshRendererProperties, rendererIndex=0}]
 }
 SharpFlare {
  bEnabled=false
  emitterUpdateScript [
   &EmitterState(
    Enable Distance Culling=true
    Life Cycle Mode=::NewEnumerator1
    Loop Delay=0.3
    Loop Duration=&Add_Float(A=$System.LifetimeAdd, B=0)
    MaxDistance=10000
    Module.bUseMaxDistance=true
    Module.bUseMinDistance=false
    Recalculate Duration Each Loop=true
    Scalability Mode=::NewEnumerator1
    UseLoopDelay=true
   )
   &SpawnBurst_Instantaneous(Spawn Count=1, Use Spawn Probability=false)
  ]
  particleSpawnScript [
   &SystemLocation(MaskSpawnGroup=false)
   &InitializeParticle(
    Color=<rgba 0 3.3333 10 1>
    Lifetime=&Add_Float(A=$System.LifetimeAdd, B=0)
    Sprite Size=<800 70>
    Sprite Size Mode=::NewEnumerator4
    UsePositionOffset=false
    Write Material Random=true
   ) {module=InitializeParticle002}
  ]
  particleUpdateScript [
   &ParticleState()
   &ScaleSpriteSize(
    Scale Factor=&Vector2DFromCurve(
     Vector2Curve=DataInterface() {propertyValues {
       XCurve {keys [
         <at 0 0>
         {interpMode=RCIM_Cubic, tangentMode=RCTM_User, time=0.95, value=0.9549}
         {arriveTangent=-23.5207, interpMode=RCIM_Cubic, leaveTangent=-23.5207, tangentMode=RCTM_User, time=0.9984, value=0}
        ]}
       YCurve {keys [<at 1 1>]}
      }}
     ScaleCurve=&Multiply_Vector2D_ByFloat(Float=$User.Radius, Vector2D=<3 3>)
    )
    Scale Sprite Size Mode=::NewEnumerator2
   )
   &ScaleColor(
    Scale Alpha=&FloatFromCurve(FloatCurve=DataInterface() {propertyValues {Curve {keys [
         {arriveTangent=-0.0603, interpMode=RCIM_Cubic, leaveTangent=-0.0603, tangentMode=RCTM_User, time=0, value=0}
         <at 0.95 1 RCIM_Cubic>
         {arriveTangent=-9.1553, tangentMode=RCTM_User, time=1, value=0}
        ]}}})
    Scale RGB=<2 2 2>
    ScaleA=true
    ScaleRGB=true
   )
   &CameraOffset(Camera Offset Amount=50)
  ]
  rendererClasses [&NiagaraSpriteRendererProperties]
  renderers [{rendererClass=&NiagaraSpriteRendererProperties}]
 }
 TinyFrags {
  bEnabled=false
  emitterUpdateScript [
   &EmitterState(
    Enable Distance Culling=true
    Life Cycle Mode=::NewEnumerator1
    Loop Behavior=::NewEnumerator1
    Loop Delay=19
    Loop Duration=1
    Module.bUseMaxDistance=true
    Module.bUseMinDistance=false
    Scalability Mode=::NewEnumerator1
    UseLoopDelay=true
   )
   &SpawnBurst_Instantaneous(Spawn Count=20, Use Spawn Probability=false, UseLoopCountLimit=false)
  ]
  particleSpawnScript [
   /SetVariables(Particles.MeshOrientation=&MakeQuatFromAxisAngle(
     Axis=&RandomRangeVector(Minimum=<0 0 0>, Maximum=<1 1 1>, Evaluation Type=::NewEnumerator1)
     AngleInDegrees=90
    )) {isSetParameters=true}
   /SetVariables(Particles.SpawnRotationVector=&RandomRangeVector(Minimum=<-0.15 -0.15 -0.15>, Maximum=<0.15 0.15 0.15>, Evaluation Type=::NewEnumerator1)) {isSetParameters=true}
   /Niagara/Modules/Spawn/Initialization/InitializeParticle(
    Color=<rgba 0 5 5 1>
    Lifetime=&RandomRangeFloat(Minimum=0.5, Maximum=1.3, Evaluation Type=::NewEnumerator1)
    Mesh Scale=&RandomRangeVector(Minimum=<0.8 0.8 0.8>, Maximum=<1.2 1.2 1.5>, Evaluation Type=::NewEnumerator1)
    Module.Write Color=true
    Module.Write Lifetime=true
    Module.Write Mass=true
    Module.Write Material Random=true
    Module.Write Position=false
    Module.Write Scale=true
    Module.Write SpriteRotation=false
    Module.Write SpriteSize=false
   ) {enabled=false}
   &ShapeLocation(
    Cylinder Height=&RandomRangeFloat(Minimum=100, Maximum=500)
    Cylinder Radius=&RandomRangeFloat(Minimum=100, Maximum=700)
    MaskSpawnGroup=false
    Offset=$System.Offset
    Offset Mode=::NewEnumerator0
    Override Seed=false
    Shape Primitive=::NewEnumerator1
    Surface Only=true
    Use Endcaps In Surface Only Mode=false
    UseOwnerScale=true
   )
   &AddVelocity(
    Module.Use Falloff Distance=false
    Override Seed=false
    Velocity Mode=::NewEnumerator1
    Velocity Speed=&RandomRangeFloat(Minimum=500, Maximum=2000)
   ) {module=AddVelocity001}
   /Niagara/Modules/Spawn/Initialization/V2/InitializeParticle(
    Lifetime=&RandomRangeFloat(Minimum=0.5, Maximum=1.3, Evaluation Type=::NewEnumerator1)
    Mesh Scale=&RandomRangeVector(Minimum=<0.8 0.8 0.8>, Maximum=<1.2 1.2 1.5>, Evaluation Type=::NewEnumerator1)
    Mesh Scale Mode=::NewEnumerator4
    UsePositionOffset=false
    Write Material Random=true
   ) {module=InitializeParticle001}
  ]
  particleUpdateScript [
   &ParticleState()
   &Drag(Drag=5, Use Linear Drag=true, Use Rotational Drag=false)
   &SolveForcesAndVelocity(Clamp Velocity=false, Limit Acceleration=false)
   &ScaleColor(
    Scale Alpha=&FloatFromCurve(FloatCurve=DataInterface() {propertyValues {Curve {keys [
         <at 0 0 RCIM_Cubic>
         {arriveTangent=1.4286, interpMode=RCIM_Cubic, leaveTangent=1.4286, time=0.2, value=1}
         {arriveTangent=-1.25, interpMode=RCIM_Cubic, leaveTangent=-1.25, time=0.7, value=1}
         <at 1 0>
        ]}}})
    ScaleA=true
    ScaleRGB=true
   )
   &UpdateMeshOrientation(Rotation Vector=$Particles.SpawnRotationVector) {module=UpdateMeshOrientation001}
   &ScaleMeshSize(Scale Factor=&VectorFromFloat(Value=&FloatFromCurve(
      FloatCurve=DataInterface() {propertyValues {Curve {keys [
          {arriveTangent=9.5937, interpMode=RCIM_Cubic, leaveTangent=9.5937, tangentMode=RCTM_User, time=0, value=0}
          <at 0.2 1>
          {arriveTangentWeight=0.2134, interpMode=RCIM_Cubic, leaveTangentWeight=0.2134, tangentMode=RCTM_User, time=0.6, value=1}
          {arriveTangent=-5.5237, interpMode=RCIM_Cubic, leaveTangent=-5.5237, tangentMode=RCTM_User, time=1, value=0}
         ]}}}
      Scale Curve=10
     )))
   /SetVariables(
    Particles.Color2=&MultiplyLinearColor(Linear Color=$Particles.Color, Scale Factor=<0 0 0 5>)
    Particles.MeshSize2:Vector3f=&Multiply_VectorByFloat(Vector=$Particles.Scale, Float=1.3)
    Particles.Size2:Vector2f=&Multiply_Vector2D(A=$Particles.SpriteSize, B=<2 2>)
   ) {isSetParameters=true}
  ]
  rendererClasses [&NiagaraMeshRendererProperties]
  renderers [
   {rendererClass=&NiagaraMeshRendererProperties, rendererIndex=0}
   {rendererClass=&NiagaraMeshRendererProperties, rendererIndex=1}
  ]
 }
 WideFlare {
  emitterUpdateScript [
   &EmitterState()
   &SpawnBurst_Instantaneous(Spawn Count=1, Use Spawn Probability=false)
  ]
  particleSpawnScript [
   /Niagara/Modules/Spawn/Initialization/InitializeParticle(
    Color=<rgba 0 0.1667 1 1>
    Lifetime=$System.LifetimeAdd
    Module.Write Color=true
    Module.Write Lifetime=true
    Module.Write Mass=true
    Module.Write Material Random=true
    Module.Write Position=false
    Module.Write Scale=false
    Module.Write SpriteRotation=false
    Module.Write SpriteSize=true
    Sprite Size=<300 200>
   ) {enabled=false}
   /Niagara/Modules/Spawn/Initialization/V2/InitializeParticle(Color=<rgba 0 0.1667 1 1>, Lifetime=$System.LifetimeAdd, UsePositionOffset=false, Write Material Random=false) {enabled=false, module=InitializeParticle001}
   &Color(Color=&MakeLinearColorFromVectorAndFloat("Vector (RGB)"=<0.3323 1.6388 2>, "Float (Alpha)"=&ClampFloat(Float=1)))
   &InitializeParticle(
    Color=<rgba 0 0.1667 1 1>
    Lifetime=&Add_Float(A=$System.LifetimeAdd, B=1)
    Sprite Size=<300 200>
    Sprite Size Mode=::NewEnumerator4
    UsePositionOffset=false
    Write Material Random=true
   ) {module=InitializeParticle002}
  ]
  particleUpdateScript [
   &ParticleState()
   &ScaleSpriteSize(
    Uniform Curve Scale=&Multiply_Float(A=5, B=$User.Radius)
    Uniform Curve Sprite Scale=DataInterface() {propertyValues {Curve {keys [{tangentMode=RCTM_User, time=0.3, value=0}, {tangentMode=RCTM_User, time=0.9, value=1}, <at 1 0>]}}}
   ) {module=ScaleSpriteSize002}
   &ScaleColor(
    Scale Alpha=0.2
    Scale RGB=&VectorFromFloat(Value=&FloatFromCurve(FloatCurve=DataInterface() {propertyValues {Curve {keys [<at 0 1 RCIM_Cubic>, {arriveTangent=6.5633, interpMode=RCIM_Cubic, leaveTangent=6.5633, tangentMode=RCTM_User, time=1, value=5}]}}}))
    ScaleA=true
    ScaleRGB=true
   )
   &DynamicMaterialParameters(
    Index 0 Param 1=&FloatFromCurve(FloatCurve=DataInterface() {propertyValues {Curve {keys [<at 0.4 0.15>,<at 0.8 0.7>]}}}, Scale Curve=4)
    Index 0 Param 2=1
    Index 0 Param 3=1
    Index 0 Param 4=1
    Module.Param0WriteEnabled=true
    Module.Param1WriteEnabled=true
    Module.Param2WriteEnabled=true
    Module.Param3WriteEnabled=true
   )
   &ScaleSpriteSize(Scale Factor=&Vector2DFromCurve(
     Vector2Curve=DataInterface() {propertyValues {
       XCurve {keys [<at 0 0.1 RCIM_Cubic>,<at 0.6 1 RCIM_Cubic>]}
       YCurve {keys [{arriveTangent=5.9554, interpMode=RCIM_Cubic, time=0, value=0.1}, <at 0.6 1 RCIM_Cubic>]}
      }}
     ScaleCurve=&Vector2DFromFloat(Value=3)
    )) {enabled=false}
   &ScaleSpriteSize(Scale Factor=&MakeVector2D(
     X=&Lerp_Float(
      A=0.2
      B=1
      Alpha=&FloatFromCurve(FloatCurve=DataInterface() {propertyValues {Curve {keys [
           {interpMode=RCIM_Cubic, tangentMode=RCTM_User, time=0, value=0}
           {arriveTangent=0.1929, interpMode=RCIM_Cubic, leaveTangent=0.1929, tangentMode=RCTM_User, time=0.8, value=0.05}
           {arriveTangentWeight=0.096, interpMode=RCIM_Cubic, tangentMode=RCTM_User, time=0.85, value=1}
          ]}}})
     )
     Y=&Lerp_Float(
      A=0.5
      B=2
      Alpha=&FloatFromCurve(FloatCurve=DataInterface() {propertyValues {Curve {keys [
           {interpMode=RCIM_Cubic, tangentMode=RCTM_User, time=0, value=0}
           {arriveTangent=0.1929, interpMode=RCIM_Cubic, leaveTangent=0.1929, tangentMode=RCTM_User, time=0.9, value=0.05}
           {arriveTangentWeight=0.096, interpMode=RCIM_Cubic, tangentMode=RCTM_User, time=0.95, value=1}
          ]}}})
     )
    )) {enabled=false, module=ScaleSpriteSize001}
   &CameraOffset(Camera Offset Amount=100)
  ]
  rendererClasses [&NiagaraSpriteRendererProperties]
  renderers [{rendererClass=&NiagaraSpriteRendererProperties}]
 }
}
summary {
 systemName=NS_LimeRock_TV_Rift
 emitters [
  {bEnabled=false, emitterName=Crack01, rendererClasses [&NiagaraMeshRendererProperties]}
  {emitterName=WideFlare, rendererClasses [&NiagaraSpriteRendererProperties]}
  {bEnabled=false, emitterName=Flare, rendererClasses [&NiagaraSpriteRendererProperties]}
  {bEnabled=false, emitterName=Embers, simTarget=GPUComputeSim, rendererClasses [&NiagaraSpriteRendererProperties]}
  {bEnabled=false, emitterName=Fragments, rendererClasses [&NiagaraMeshRendererProperties]}
  {emitterName=Bolt, rendererClasses [&NiagaraMeshRendererProperties]}
  {bEnabled=false, emitterName=Crack02, rendererClasses [&NiagaraMeshRendererProperties]}
  {emitterName=Light001, rendererClasses [&NiagaraLightRendererProperties]}
  {bEnabled=false, emitterName=Corona, rendererClasses [&NiagaraSpriteRendererProperties]}
  {emitterName=Rays, rendererClasses [&NiagaraMeshRendererProperties]}
  {bEnabled=false, emitterName=SharpFlare, rendererClasses [&NiagaraSpriteRendererProperties]}
  {emitterName=Corona001, rendererClasses [&NiagaraSpriteRendererProperties]}
  {
   emitterName=FlareLightShockwave
   rendererClasses [
    &NiagaraSpriteRendererProperties
    &NiagaraLightRendererProperties
    &NiagaraMeshRendererProperties
   ]
  }
  {bEnabled=false, emitterName=TinyFrags, rendererClasses [&NiagaraMeshRendererProperties]}
 ]
 userVariables [
  User.bIsFrontEnd {type=NiagaraBool}
  User.bIsFrontEndPreview {type=NiagaraBool}
  User.Lifetime {type=NiagaraFloat, defaultValue=11, description=""}
  User.Open {type=NiagaraFloat, defaultValue=1, description=""}
  User.Radius {type=NiagaraFloat, defaultValue=10, description=""}
  User.TestModel {type=NiagaraBool}
 ]
}
